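# Sound cue for ejecting a drum magazine during weapon handling, classed as weapon foley.
# A random node picks one of four eject wave variants with equal weight, then a modulator applies slight pitch and volume variation.
# It plays at reduced volume through the weapon reload attenuation and overrides concurrency settings.
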SoundCue /Game/Sounds/Fort_Weapons/Magazines/Magazine_Drum_Eject_Cue
root=Modulator_0
nodes [
 {id=Modulator_0, children [Random_0], properties {pitchMax=1.1, pitchMin=0.9, volumeMax=1.1, volumeMin=0.9}}
 {id=Random_0, children [WavePlayer_0,WavePlayer_1,WavePlayer_2,WavePlayer_3], properties {weights [1,1,1,1]}}
 {id=WavePlayer_0, properties {soundWaveAssetPtr=&/Magazine_Drum_Eject_04}}
 {id=WavePlayer_1, properties {soundWaveAssetPtr=&/Magazine_Drum_Eject_03}}
 {id=WavePlayer_2, properties {soundWaveAssetPtr=&/Magazine_Drum_Eject_02}}
 {id=WavePlayer_3, properties {soundWaveAssetPtr=&/Magazine_Drum_Eject_01}}
]
properties {
 attenuationSettings=&Fort_Weapon_Reload_Att
 bOverrideConcurrency=true
 maxDistance=2250
 soundClassObject=&Fort_WeaponFoley
 volumeMultiplier=0.301
}
