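# Sound cue for inserting a drum magazine during a weapon reload, classed as weapon foley.
# A random node picks one of four equally weighted insert recordings, and a root modulator slightly varies pitch and volume for variation between plays.
# The cue uses the shared weapon reload attenuation, plays at reduced volume and overrides concurrency.
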SoundCue /Game/Sounds/Fort_Weapons/Magazines/Magazine_Drum_Insert_Cue
root=Modulator_0
nodes [
 {id=Modulator_0, children [Random_0], properties {pitchMax=1.1, pitchMin=0.9, volumeMax=1.1, volumeMin=0.9}}
 {id=Random_0, children [WavePlayer_5,WavePlayer_6,WavePlayer_7,WavePlayer_8], properties {weights [1,1,1,1]}}
 {id=WavePlayer_5, properties {soundWaveAssetPtr=&/Magazine_Drum_Insert_01}}
 {id=WavePlayer_6, properties {soundWaveAssetPtr=&/Magazine_Drum_Insert_02}}
 {id=WavePlayer_7, properties {soundWaveAssetPtr=&/Magazine_Drum_Insert_03}}
 {id=WavePlayer_8, properties {soundWaveAssetPtr=&/Magazine_Drum_Insert_04}}
]
properties {
 attenuationSettings=&Fort_Weapon_Reload_Att
 bOverrideConcurrency=true
 maxDistance=2250
 soundClassObject=&Fort_WeaponFoley
 volumeMultiplier=0.401
}
