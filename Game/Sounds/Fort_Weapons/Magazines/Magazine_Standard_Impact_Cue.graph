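# Sound cue for a standard weapon magazine impact, using the weapon-drop attenuation and the loot-collect sound class.
# A Random node picks one of six equally weighted impact wave variations.
# A Modulator at the root applies slight random pitch and volume variation to the pick.
# The cue plays at a reduced overall volume and overrides its concurrency settings.
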SoundCue /Game/Sounds/Fort_Weapons/Magazines/Magazine_Standard_Impact_Cue
root=Modulator_0
nodes [
 {id=Modulator_0, children [Random_0], properties {pitchMax=1.1, pitchMin=0.9, volumeMax=1.1, volumeMin=0.9}}
 {
  id=Random_0
  children [WavePlayer_5, WavePlayer_6, WavePlayer_7, WavePlayer_8, WavePlayer_9, WavePlayer_10]
  properties {weights [1,1,1,1,1,1]}
 }
 {id=WavePlayer_10, properties {soundWaveAssetPtr=&/Magazine_Standard_Impact_06}}
 {id=WavePlayer_5, properties {soundWaveAssetPtr=&/Magazine_Standard_Impact_01}}
 {id=WavePlayer_6, properties {soundWaveAssetPtr=&/Magazine_Standard_Impact_02}}
 {id=WavePlayer_7, properties {soundWaveAssetPtr=&/Magazine_Standard_Impact_03}}
 {id=WavePlayer_8, properties {soundWaveAssetPtr=&/Magazine_Standard_Impact_04}}
 {id=WavePlayer_9, properties {soundWaveAssetPtr=&/Magazine_Standard_Impact_05}}
]
properties {
 attenuationSettings=&Fort_Weapon_Drop_Att
 bOverrideConcurrency=true
 maxDistance=3000
 soundClassObject=&Fort_Loot_Collect
 volumeMultiplier=0.25
}
